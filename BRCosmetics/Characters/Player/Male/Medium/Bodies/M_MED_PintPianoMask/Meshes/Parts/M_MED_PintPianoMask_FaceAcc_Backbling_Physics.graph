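# Physics asset for the backbling variant of the PintPianoMask male medium face accessory, defining capsule bodies for simulating its hood.
# Mirrored front chains and shorter rear chains, plus a rear middle chain, run from small base bodies, alongside larger spine and clavicle capsules.
# Angular limits are tighter at the rear than the front.
# Cross-links join the lowercase-c front tips to each other and the rear-b tips to the middle tip.
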PhysicsAsset /BRCosmetics/Characters/Player/Male/Medium/Bodies/M_MED_PintPianoMask/Meshes/Parts/M_MED_PintPianoMask_FaceAcc_Backbling_Physics
bodies {
 clavicle_l [{shapeType=Capsule, length=15.2931, radius=9.4192, shapeName=clavicle_l_capsule, center=<8.836 -0.3916 -0.7726>, rotation=<rot 0 -90 98.5569>}]
 clavicle_r [{shapeType=Capsule, length=15.7531, radius=8.9592, shapeName=clavicle_r_capsule, center=<-8.8637 0 0.5887>, rotation=<rot 0 -90 -81.4431>}]
 dyn_hood_front_a_01_l [{shapeType=Capsule, length=4, radius=3, shapeName=dyn_hood_front_a_01_l_capsule, center=<2.5 3.3 0>, rotation=<rot 0 -90 -15>}]
 dyn_hood_front_a_01_r [{shapeType=Capsule, length=4, radius=3, shapeName=dyn_hood_front_a_01_r_capsule, center=<-2.5 -3.3 0>, rotation=<rot 0 -90 165>}]
 dyn_hood_front_a_02_l [{shapeType=Capsule, length=4, radius=3, shapeName=dyn_hood_front_a_02_l_capsule, center=<2.5 4.4 0>, rotation=<rot -2.9838 -89.1997 -15.0208>}]
 dyn_hood_front_a_02_r [{
   shapeType=Capsule
   length=4
   radius=3
   shapeName=dyn_hood_front_a_02_r_capsule
   center=<-2.5 -4.4 0>
   rotation=<rot -2.9838 -89.1997 164.9792>
  }]
 dyn_hood_front_a_03_l [{shapeType=Capsule, length=4, radius=3, shapeName=dyn_hood_front_a_03_l_capsule, center=<2.5 5 0>, rotation=<rot -2.5983 -89.3032 -15.0157>}]
 dyn_hood_front_a_03_r [{shapeType=Capsule, length=4, radius=3, shapeName=dyn_hood_front_a_03_r_capsule, center=<-2.5 -5 0>, rotation=<rot -2.5983 -89.3032 164.9843>}]
 dyn_hood_front_a_l_base [{shapeType=Capsule, length=0.505, radius=0.505, shapeName=dyn_hood_front_a_l_base_capsule, rotation=<rot 0 -90 0>}]
 dyn_hood_front_a_r_base [{shapeType=Capsule, length=0.505, radius=0.505, shapeName=dyn_hood_front_a_r_base_capsule, rotation=<rot 0 -90 180>}]
 dyn_hood_front_b_01_l [{shapeType=Capsule, length=4, radius=3, shapeName=dyn_hood_front_b_01_l_capsule, center=<2.5 2.5 0>, rotation=<rot 0 -90 -5>}]
 dyn_hood_front_b_01_r [{shapeType=Capsule, length=4, radius=3, shapeName=dyn_hood_front_b_01_r_capsule, center=<-2.5 -2.5 0>, rotation=<rot 0 -90 175>}]
 dyn_hood_front_b_02_l [{shapeType=Capsule, length=4, radius=3, shapeName=dyn_hood_front_b_02_l_capsule, center=<2.5 3.6 0>, rotation=<rot 0 -90 -5>}]
 dyn_hood_front_b_02_r [{shapeType=Capsule, length=4, radius=3, shapeName=dyn_hood_front_b_02_r_capsule, center=<-2.5 -3.6 0>, rotation=<rot 0 -90 175>}]
 dyn_hood_front_b_03_l [{shapeType=Capsule, length=4, radius=3, shapeName=dyn_hood_front_b_03_l_capsule, center=<2.5 4.6057 0>, rotation=<rot 0 -90 -5>}]
 dyn_hood_front_b_03_r [{shapeType=Capsule, length=4, radius=3, shapeName=dyn_hood_front_b_03_r_capsule, center=<-2.5 -4.6057 0>, rotation=<rot 0 -90 175>}]
 dyn_hood_front_b_l_base [{shapeType=Capsule, length=0.505, radius=0.505, shapeName=dyn_hood_front_b_l_base_capsule, rotation=<rot 0 -90 0>}]
 dyn_hood_front_b_r_base [{shapeType=Capsule, length=0.505, radius=0.505, shapeName=dyn_hood_front_b_r_base_capsule, rotation=<rot 0 -90 180>}]
 dyn_hood_front_c_01_l [{shapeType=Capsule, length=4, radius=3, shapeName=dyn_hood_front_c_01_l_capsule, center=<2.5 1 0>, rotation=<rot 0 -90 -5>}]
 dyn_hood_front_c_01_r [{shapeType=Capsule, length=4, radius=3, shapeName=dyn_hood_front_c_01_r_capsule, center=<-2.5 -1 0>, rotation=<rot 0 -90 175>}]
 dyn_hood_front_c_02_l [{shapeType=Capsule, length=4, radius=3, shapeName=dyn_hood_front_c_02_l_capsule, center=<2.5 1.85 0>, rotation=<rot 0 -90 -5>}]
 dyn_hood_front_c_02_r [{shapeType=Capsule, length=4, radius=3, shapeName=dyn_hood_front_c_02_r_capsule, center=<-2.5 -1.85 0>, rotation=<rot 0 -90 175>}]
 dyn_hood_front_c_03_l [{
   shapeType=Capsule
   length=4
   radius=3
   shapeName=dyn_hood_front_c_03_l_capsule
   center=<2.4937 3.9008 0.0717>
   rotation=<rot 10.5277 -90.9314 -5.0858>
  }]
 dyn_hood_front_c_03_r [{
   shapeType=Capsule
   length=4
   radius=3
   shapeName=dyn_hood_front_c_03_r_capsule
   center=<-2.4937 -3.9008 -0.0717>
   rotation=<rot 10.5277 -90.9314 174.9142>
  }]
 dyn_hood_front_c_l_base [{shapeType=Capsule, length=0.505, radius=0.505, shapeName=dyn_hood_front_c_l_base_capsule, rotation=<rot 0 -90 0>}]
 dyn_hood_front_c_r_base [{shapeType=Capsule, length=0.505, radius=0.505, shapeName=dyn_hood_front_c_r_base_capsule, rotation=<rot 0 -90 180>}]
 dyn_hood_rear_a_01_l [{shapeType=Capsule, length=4, radius=3, shapeName=dyn_hood_rear_a_01_l_capsule, center=<3 2.2 0>, rotation=<rot 0 90 0>}]
 dyn_hood_rear_a_01_r [{shapeType=Capsule, length=3.975, radius=3.025, shapeName=dyn_hood_rear_a_01_r_capsule, center=<-3 -2.2 0>, rotation=<rot 0 90 -180>}]
 dyn_hood_rear_a_02_l [{shapeType=Capsule, length=4, radius=3, shapeName=dyn_hood_rear_a_02_l_capsule, center=<2.4547 4 0>, rotation=<rot 0 90 0>}]
 dyn_hood_rear_a_02_r [{shapeType=Capsule, length=4, radius=3, shapeName=dyn_hood_rear_a_02_r_capsule, center=<-2.2003 -2.686 0>, rotation=<rot 6.2406 74.5232 180>}]
 dyn_hood_rear_a_l_base [{shapeType=Capsule, length=0.505, radius=0.505, shapeName=dyn_hood_rear_a_l_base_capsule, rotation=<rot 0 -90 0>}]
 dyn_hood_rear_a_r_base [{shapeType=Capsule, length=0.505, radius=0.505, shapeName=dyn_hood_rear_a_r_base_capsule, rotation=<rot 0 -90 180>}]
 dyn_hood_rear_b_01_l [{shapeType=Capsule, length=3.647, radius=3.622, shapeName=dyn_hood_rear_b_01_l_capsule, center=<3 0 0>, rotation=<rot 0 90 0>}]
 dyn_hood_rear_b_01_r [{shapeType=Capsule, length=3.647, radius=3.622, shapeName=dyn_hood_rear_b_01_r_capsule, center=<-3 0 0>, rotation=<rot 0 90 -180>}]
 dyn_hood_rear_b_02_l [{
   shapeType=Capsule
   length=3.8169
   radius=3.1831
   shapeName=dyn_hood_rear_b_02_l_capsule
   center=<2.6704 3.7231 0>
   rotation=<rot 4.1046 107.1978 0.9662>
  }]
 dyn_hood_rear_b_02_r [{
   shapeType=Capsule
   length=3.7158
   radius=3.2842
   shapeName=dyn_hood_rear_b_02_r_capsule
   center=<-2.6704 -3.7231 0>
   rotation=<rot 4.1046 107.1978 -179.0338>
  }]
 dyn_hood_rear_b_l_base [{shapeType=Capsule, length=0.505, radius=0.505, shapeName=dyn_hood_rear_b_l_base_capsule, rotation=<rot 0 -90 0>}]
 dyn_hood_rear_b_r_base [{shapeType=Capsule, length=0.505, radius=0.505, shapeName=dyn_hood_rear_b_r_base_capsule, rotation=<rot 0 -90 180>}]
 dyn_hood_rear_mid_01 [{shapeType=Capsule, length=5, radius=4, shapeName=dyn_hood_rear_mid_01_capsule, center=<2 1.4905 0>, rotation=<rot 0 -90 0>}]
 dyn_hood_rear_mid_02 [{shapeType=Capsule, length=4.7251, radius=4.2749, shapeName=dyn_hood_rear_mid_02_capsule, center=<1.1383 3.3623 0>, rotation=<rot 0 -90 0>}]
 dyn_hood_rear_mid_base [{shapeType=Capsule, length=0.505, radius=0.505, shapeName=dyn_hood_rear_mid_base_capsule, rotation=<rot 0 -90 0>}]
 spine_04 [{shapeType=Capsule, length=13.0595, radius=11.8662, shapeName=spine_04_capsule, center=<3.8917 4.095 0>, rotation=<rot 0 -90 0>}]
 spine_05 [{shapeType=Capsule, length=14.7885, radius=12.1259, shapeName=spine_05_capsule, center=<2.322 0 0>, rotation=<rot 0 -90 0>}]
}
constraints [
 {bone1Name=dyn_hood_front_a_01_l, bone2Name=dyn_hood_front_a_l_base, swing1LimitDegrees=12, swing2LimitDegrees=12, twistLimitDegrees=12}
 {bone1Name=dyn_hood_front_a_02_l, bone2Name=dyn_hood_front_a_01_l, swing1LimitDegrees=12, swing2LimitDegrees=12, twistLimitDegrees=12}
 {bone1Name=dyn_hood_front_a_03_l, bone2Name=dyn_hood_front_a_02_l, swing1LimitDegrees=12, swing2LimitDegrees=12, twistLimitDegrees=12}
 {bone1Name=dyn_hood_front_b_01_l, bone2Name=dyn_hood_front_b_l_base, swing1LimitDegrees=12, swing2LimitDegrees=12, twistLimitDegrees=12}
 {bone1Name=dyn_hood_front_b_02_l, bone2Name=dyn_hood_front_b_01_l, swing1LimitDegrees=12, swing2LimitDegrees=12, twistLimitDegrees=12}
 {bone1Name=dyn_hood_front_b_03_l, bone2Name=dyn_hood_front_b_02_l, swing1LimitDegrees=12, swing2LimitDegrees=12, twistLimitDegrees=12}
 {bone1Name=dyn_hood_front_c_01_l, bone2Name=dyn_hood_front_c_l_base, swing1LimitDegrees=12, swing2LimitDegrees=12, twistLimitDegrees=12}
 {bone1Name=dyn_hood_front_c_02_l, bone2Name=dyn_hood_front_c_01_l, swing1LimitDegrees=12, swing2LimitDegrees=12, twistLimitDegrees=12}
 {bone1Name=dyn_hood_front_c_03_l, bone2Name=dyn_hood_front_c_02_l, swing1LimitDegrees=12, swing2LimitDegrees=12, twistLimitDegrees=12}
 {bone1Name=dyn_hood_front_a_01_r, bone2Name=dyn_hood_front_a_r_base, swing1LimitDegrees=12, swing2LimitDegrees=12, twistLimitDegrees=12}
 {bone1Name=dyn_hood_front_a_02_r, bone2Name=dyn_hood_front_a_01_r, swing1LimitDegrees=12, swing2LimitDegrees=12, twistLimitDegrees=12}
 {bone1Name=dyn_hood_front_a_03_r, bone2Name=dyn_hood_front_a_02_r, swing1LimitDegrees=12, swing2LimitDegrees=12, twistLimitDegrees=12}
 {bone1Name=dyn_hood_front_b_01_r, bone2Name=dyn_hood_front_b_r_base, swing1LimitDegrees=12, swing2LimitDegrees=12, twistLimitDegrees=12}
 {bone1Name=dyn_hood_front_b_02_r, bone2Name=dyn_hood_front_b_01_r, swing1LimitDegrees=12, swing2LimitDegrees=12, twistLimitDegrees=12}
 {bone1Name=dyn_hood_front_b_03_r, bone2Name=dyn_hood_front_b_02_r, swing1LimitDegrees=12, swing2LimitDegrees=12, twistLimitDegrees=12}
 {bone1Name=dyn_hood_front_c_01_r, bone2Name=dyn_hood_front_c_r_base, swing1LimitDegrees=12, swing2LimitDegrees=12, twistLimitDegrees=12}
 {bone1Name=dyn_hood_front_c_02_r, bone2Name=dyn_hood_front_c_01_r, swing1LimitDegrees=12, swing2LimitDegrees=12, twistLimitDegrees=12}
 {bone1Name=dyn_hood_front_c_03_r, bone2Name=dyn_hood_front_c_02_r, swing1LimitDegrees=12, swing2LimitDegrees=12, twistLimitDegrees=12}
 {bone1Name=dyn_hood_rear_b_01_l, bone2Name=dyn_hood_rear_b_l_base, swing1LimitDegrees=4, swing2LimitDegrees=4, twistLimitDegrees=4}
 {bone1Name=dyn_hood_rear_b_02_l, bone2Name=dyn_hood_rear_b_01_l, swing1LimitDegrees=4, swing2LimitDegrees=4, twistLimitDegrees=4}
 {bone1Name=dyn_hood_rear_a_01_l, bone2Name=dyn_hood_rear_a_l_base, swing1LimitDegrees=4, swing2LimitDegrees=4, twistLimitDegrees=4}
 {bone1Name=dyn_hood_rear_a_02_l, bone2Name=dyn_hood_rear_a_01_l, swing1LimitDegrees=4, swing2LimitDegrees=4, twistLimitDegrees=4}
 {bone1Name=dyn_hood_rear_a_01_r, bone2Name=dyn_hood_rear_a_r_base, swing1LimitDegrees=4, swing2LimitDegrees=4, twistLimitDegrees=4}
 {bone1Name=dyn_hood_rear_a_02_r, bone2Name=dyn_hood_rear_a_01_r, swing1LimitDegrees=4, swing2LimitDegrees=4, twistLimitDegrees=4}
 {
  bone1Name=dyn_hood_rear_b_r_base
  bone2Name=spine_05
  swing1LimitDegrees=45
  swing1Motion=Free
  swing2LimitDegrees=45
  swing2Motion=Free
  twistLimitDegrees=45
  twistMotion=Free
 }
 {bone1Name=dyn_hood_rear_b_01_r, bone2Name=dyn_hood_rear_b_r_base, swing1LimitDegrees=4, swing2LimitDegrees=4, twistLimitDegrees=4}
 {bone1Name=dyn_hood_rear_b_02_r, bone2Name=dyn_hood_rear_b_01_r, swing1LimitDegrees=4, swing2LimitDegrees=4, twistLimitDegrees=4}
 {
  bone1Name=dyn_hood_rear_mid_base
  bone2Name=spine_05
  swing1LimitDegrees=45
  swing1Motion=Free
  swing2LimitDegrees=45
  swing2Motion=Free
  twistLimitDegrees=45
  twistMotion=Free
 }
 {bone1Name=dyn_hood_rear_mid_01, bone2Name=dyn_hood_rear_mid_base, swing1LimitDegrees=4, swing2LimitDegrees=4, twistLimitDegrees=4}
 {bone1Name=dyn_hood_rear_mid_02, bone2Name=dyn_hood_rear_mid_01, swing1LimitDegrees=4, swing2LimitDegrees=4, twistLimitDegrees=4}
 {
  bone1Name=spine_05
  bone2Name=spine_04
  swing1LimitDegrees=45
  swing1Motion=Free
  swing2LimitDegrees=45
  swing2Motion=Free
  twistLimitDegrees=45
  twistMotion=Free
 }
 {
  bone1Name=dyn_hood_rear_b_02_l
  bone2Name=dyn_hood_rear_mid_02
  swing1LimitDegrees=45
  swing1Motion=Free
  swing2LimitDegrees=45
  swing2Motion=Free
  twistLimitDegrees=45
  twistMotion=Free
 }
 {
  bone1Name=dyn_hood_rear_b_02_r
  bone2Name=dyn_hood_rear_mid_02
  swing1LimitDegrees=45
  swing1Motion=Free
  swing2LimitDegrees=45
  swing2Motion=Free
  twistLimitDegrees=45
  twistMotion=Free
 }
 {bone1Name=dyn_hood_front_c_03_r, bone2Name=dyn_hood_front_c_03_l, swing1LimitDegrees=20, swing2LimitDegrees=20, twistLimitDegrees=20}
 {bone1Name=dyn_hood_front_c_03_l, bone2Name=dyn_hood_front_c_03_r, swing1LimitDegrees=20, swing2LimitDegrees=20, twistLimitDegrees=20}
]
