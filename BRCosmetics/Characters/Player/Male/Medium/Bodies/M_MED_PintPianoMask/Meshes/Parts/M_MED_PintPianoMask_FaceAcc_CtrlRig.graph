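# Control Rig for the PintPianoMask male medium face accessory, posing the hood's dynamic base bones against body motion.
# It matches the hood base to head rotation, then spherical pose readers store head, clavicle and upper-arm weights.
# Weighted parent constraints anchor the rear hood bones to the torso and arms.
# The weights then drive corrective, mostly mirrored transform offsets, some curve-remapped or Kalman-smoothed, with extra fixes when IsFrontendIdle is true.
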ControlRigBlueprint /BRCosmetics/Characters/Player/Male/Medium/Bodies/M_MED_PintPianoMask/Meshes/Parts/M_MED_PintPianoMask_FaceAcc_CtrlRig
graphs [
 RigVMFunctionLibrary {links [], nodes []}
 RigVMFunction_Sequence_2_ContainedGraph {nodes [
   n1 = RigVMFunctionEntryNode()
   RigVMFunction_Sequence = &RigVMFunction_Sequence(ExecuteContext=@n1.ExecuteContext)
   n2 = &RigVMFunction_Sequence(ExecuteContext=@RigVMFunction_Sequence.B)
   RigVMFunctionReturnNode(A=@RigVMFunction_Sequence.A, B=@n2.A, C=@n2.B)
  ]}
 RigVMModel {nodes [
   RigUnit_BeginExecution = &RigUnit_BeginExecution()
   RigUnit_SetRotation = &RigUnit_SetRotation(
    Item=(Type=Bone,Name="dyn_hood_base")
    Space=GlobalSpace
    Weight=1
    bInitial=False
    ExecutePin=@RigUnit_BeginExecution.ExecutePin
    Value=&RigUnit_GetTransform(Item=(Type=Bone,Name="head"), bInitial=False).Transform.Rotation
   )
   > RigVMFunction_Sequence = RigVMAggregateNode()
   &RigUnit_ParentConstraint(
    AdvancedSettings=(RotationOrderForFilter=XZY)
    Child=(Type=Bone,Name="dyn_hood_rear_b_l_base")
    Parents="((Item=(Type=Bone,Name=\"neck_02\"),Weight=0.0074),(Item=(Type=Bone,Name=\"spine_04\"),Weight=0.0114),(Item=(Type=Bone,Name=\"deform_upperarm_up_l\"),Weight=0.0249),(Item=(Type=Bone,Name=\"upperarm_twist_01_l\"),Weight=0.0273),(Item=(Type=Bone,Name=\"spine_03\"),Wei ...(truncated)"
    Weight=1
    ExecutePin=@RigVMFunction_Sequence.B
   )
   > &RigUnit_ParentConstraint(
    AdvancedSettings=(RotationOrderForFilter=XZY)
    Child=(Type=Bone,Name="dyn_hood_rear_mid_base")
    Parents="((Item=(Type=Bone,Name=\"neck_01\"),Weight=0.0024),(Item=(Type=Bone,Name=\"neck_02\"),Weight=0.0107),(Item=(Type=Bone,Name=\"spine_04\"),Weight=0.0277),(Item=(Type=Bone,Name=\"clavicle_r\"),Weight=0.0684),(Item=(Type=Bone,Name=\"clavicle_l\"),Weight=0.084),(Item=(Ty ...(truncated)"
    Weight=1
   )
   > &RigUnit_ParentConstraint(
    AdvancedSettings=(RotationOrderForFilter=XZY)
    Child=(Type=Bone,Name="dyn_hood_rear_b_r_base")
    Parents="((Item=(Type=Bone,Name=\"neck_02\"),Weight=0.0024),(Item=(Type=Bone,Name=\"spine_04\"),Weight=0.0179),(Item=(Type=Bone,Name=\"upperarm_twist_01_r\"),Weight=0.0374),(Item=(Type=Bone,Name=\"deform_upperarm_up_r\"),Weight=0.0407),(Item=(Type=Bone,Name=\"spine_03\"),Wei ...(truncated)"
    Weight=1
   )
   > &RigUnit_ParentConstraint(
    AdvancedSettings=(RotationOrderForFilter=XZY)
    Child=(Type=Bone,Name="dyn_hood_rear_a_l_base")
    Parents="((Item=(Type=Bone,Name=\"spine_04\"),Weight=0.0007),(Item=(Type=Bone,Name=\"neck_02\"),Weight=0.0026),(Item=(Type=Bone,Name=\"spine_03\"),Weight=0.0283),(Item=(Type=Bone,Name=\"upperarm_twist_01_l\"),Weight=0.0458),(Item=(Type=Bone,Name=\"deform_upperarm_up_l\"),Wei ...(truncated)"
    Weight=1
   )
   > &RigUnit_ParentConstraint(
    AdvancedSettings=(RotationOrderForFilter=XZY)
    Child=(Type=Bone,Name="dyn_hood_rear_a_r_base")
    Parents="((Item=(Type=Bone,Name=\"deform_upperarm_fr_r\"),Weight=0.0036),(Item=(Type=Bone,Name=\"spine_04\"),Weight=0.0072),(Item=(Type=Bone,Name=\"spine_03\"),Weight=0.0337),(Item=(Type=Bone,Name=\"upperarm_twist_01_r\"),Weight=0.0651),(Item=(Type=Bone,Name=\"deform_uppera ...(truncated)"
    Weight=1
   )
   Modify Transforms with Mirror_4 = RigVMFunctionReferenceNode(
    Base Modify Data="((Item=(Type=Bone,Name=\"dyn_hood_rear_a_l_base\"),Transform=(Rotation=(X=0,Y=0,Z=0.0349,W=0.9994),Translation=(X=-1,Y=1.5,Z=1),Scale3D=(X=1,Y=1,Z=1))),(Item=(Type=Bone,Name=\"dyn_hood_rear_b_l_base\"),Transform=(Rotation=(X=0,Y=0,Z=0.0436,W=0.999),Translation ...(truncated)"
    Mirror Modify Data="((Item=(Type=Bone,Name=\"dyn_hood_rear_a_r_base\"),Transform=(Rotation=(X=0,Y=0,Z=0,W=1),Translation=(X=0,Y=0,Z=0),Scale3D=(X=1,Y=1,Z=1))),(Item=(Type=Bone,Name=\"dyn_hood_rear_b_r_base\"),Transform=(Rotation=(X=0,Y=0,Z=0,W=1),Translation=(X=0,Y=0,Z=0),Scale3D ...(truncated)"
    Weight=RigVMVariableNode() {variable=SPR Upperarm L Fwd}.Value
    "Mirror Weight "=RigVMVariableNode() {variable=SPR Upperarm R Fwd}.Value
    ExecuteContext=@RigVMFunction_Sequence.C
   )
   > Modify Transforms with Mirror_2 = RigVMFunctionReferenceNode(
    Base Modify Data="((Item=(Type=Bone,Name=\"dyn_hood_rear_a_l_base\"),Transform=(Rotation=(X=0.1285,Y=-0.2965,Z=-0.1926,W=0.9266),Translation=(X=0,Y=-5,Z=-1),Scale3D=(X=1,Y=1,Z=1))),(Item=(Type=Bone,Name=\"dyn_hood_rear_b_l_base\"),Transform=(Rotation=(X=0,Y=0,Z=0,W=1),Translati ...(truncated)"
    Mirror Modify Data="((Item=(Type=Bone,Name=\"dyn_hood_rear_a_r_base\"),Transform=(Rotation=(X=0,Y=0,Z=0,W=1),Translation=(X=0,Y=0,Z=0),Scale3D=(X=1,Y=1,Z=1))),(Item=(Type=Bone,Name=\"dyn_hood_rear_b_r_base\"),Transform=(Rotation=(X=0,Y=0,Z=0,W=1),Translation=(X=0,Y=0,Z=0),Scale3D ...(truncated)"
    Weight=RigVMVariableNode() {variable=SPR Upperarm L Raised}.Value
    "Mirror Weight "=RigVMVariableNode() {variable=SPR Upperarm R Raised}.Value
   )
   > &RigUnit_ModifyTransforms(
    ItemToModify="((Item=(Type=Bone,Name=\"dyn_hood_rear_b_l_base\"),Transform=(Rotation=(X=-0.0231,Y=-0.2449,Z=0.0064,W=0.9692),Translation=(X=0.0322,Y=0.7744,Z=0.5063),Scale3D=(X=1,Y=1,Z=1))),(Item=(Type=Bone,Name=\"dyn_hood_rear_b_r_base\"),Transform=(Rotation=(X=0.0504,Y=-0 ...(truncated)"
    WeightMinimum=0
    Weight=RigVMVariableNode() {variable=SPR Head Back}.Value
   )
   n1 = RigVMVariableNode() {variable=SPR Clavicle L Back}
   n2 = &RigVMFunction_AnimEvalRichCurve(
    Curve="(EditorCurveData=(Keys=((InterpMode=RCIM_Cubic),(InterpMode=RCIM_Cubic,TangentMode=RCTM_User,Time=0.7563,Value=0.1976,ArriveTangent=0.4865,LeaveTangent=0.4865),(InterpMode=RCIM_Cubic,TangentMode=RCTM_User,Time=1,Value=1,ArriveTangent=2.3851,LeaveTangent=2. ...(truncated)"
    SourceMinimum=0
    TargetMinimum=0
    Value=@n1.Value
   )
   KalmanFilter = &RigVMFunction_KalmanFloat(BufferSize=4, Value=@n2.Result) {template="KalmanFilter::Execute(in Value,in BufferSize,out Result)"}
   n3 = RigVMVariableNode() {variable=SPR Clavicle R Back}
   n4 = &RigVMFunction_AnimEvalRichCurve(
    Curve="(EditorCurveData=(Keys=((InterpMode=RCIM_Cubic),(InterpMode=RCIM_Cubic,TangentMode=RCTM_User,Time=0.7563,Value=0.1976,ArriveTangent=0.4865,LeaveTangent=0.4865),(InterpMode=RCIM_Cubic,TangentMode=RCTM_User,Time=1,Value=1,ArriveTangent=2.3851,LeaveTangent=2. ...(truncated)"
    SourceMinimum=0
    TargetMinimum=0
    Value=@n3.Value
   )
   > Modify Transforms with Mirror_3 = RigVMFunctionReferenceNode(
    Base Modify Data="((Item=(Type=Bone,Name=\"dyn_hood_rear_a_l_base\"),Transform=(Rotation=(X=0,Y=-0.1736,Z=0,W=0.9848),Translation=(X=-3,Y=4,Z=0),Scale3D=(X=1,Y=1,Z=1))),(Item=(Type=Bone,Name=\"dyn_hood_rear_b_l_base\"),Transform=(Rotation=(X=0.1555,Y=0.1146,Z=0.1555,W=0.9688),T ...(truncated)"
    Mirror Modify Data="((Item=(Type=Bone,Name=\"dyn_hood_rear_a_r_base\"),Transform=(Rotation=(X=0,Y=0,Z=0,W=1),Translation=(X=0,Y=0,Z=0),Scale3D=(X=1,Y=1,Z=1))),(Item=(Type=Bone,Name=\"dyn_hood_rear_b_r_base\"),Transform=(Rotation=(X=0,Y=0,Z=0,W=1),Translation=(X=0,Y=0,Z=0),Scale3D ...(truncated)"
    Weight=@KalmanFilter.Result
    "Mirror Weight "=&RigVMFunction_KalmanFloat(BufferSize=4, Value=@n4.Result) {template="KalmanFilter::Execute(in Value,in BufferSize,out Result)"}.Result
   )
   > &RigUnit_ModifyTransforms(
    ItemToModify=((Item=(Type=Bone,Name="dyn_hood_rear_mid_base"),Transform=(Rotation=(X=0,Y=0,Z=0.2079,W=0.9781),Translation=(X=0,Y=0,Z=0),Scale3D=(X=1,Y=1,Z=1))))
    WeightMinimum=0
    Weight=@n1.Value
   )
   > &RigUnit_ModifyTransforms(
    ItemToModify=((Item=(Type=Bone,Name="dyn_hood_rear_mid_base"),Transform=(Rotation=(X=0,Y=0,Z=0.2079,W=0.9781),Translation=(X=0,Y=0,Z=0),Scale3D=(X=1,Y=1,Z=1))))
    WeightMinimum=0
    Weight=@n3.Value
   )
   > Modify Transforms with Mirror_5 = RigVMFunctionReferenceNode(
    Base Modify Data="((Item=(Type=Bone,Name=\"dyn_hood_rear_a_l_base\"),Transform=(Rotation=(X=0.0335,Y=0.1819,Z=0.1594,W=0.9697),Translation=(X=-1,Y=3,Z=-2),Scale3D=(X=1,Y=1,Z=1))),(Item=(Type=Bone,Name=\"dyn_hood_rear_b_l_base\"),Transform=(Rotation=(X=0.1387,Y=0.1981,Z=0.1387,W ...(truncated)"
    Mirror Modify Data="((Item=(Type=Bone,Name=\"dyn_hood_rear_a_r_base\"),Transform=(Rotation=(X=0,Y=0,Z=0,W=1),Translation=(X=0,Y=0,Z=0),Scale3D=(X=1,Y=1,Z=1))),(Item=(Type=Bone,Name=\"dyn_hood_rear_b_r_base\"),Transform=(Rotation=(X=0,Y=0,Z=0,W=1),Translation=(X=0,Y=0,Z=0),Scale3D ...(truncated)"
    Weight=RigVMVariableNode() {variable=SPR Clavicle L Raise Side}.Value
    "Mirror Weight "=RigVMVariableNode() {variable=SPR Clavicle R Raise Side}.Value
   )
   Add = &RigVMFunction_MathDoubleAdd(A=RigVMVariableNode() {variable=SPR Clavicle L Raise Side}.Value, B=RigVMVariableNode() {variable=SPR Upperarm L Raised}.Value)
   Multiply = &RigVMFunction_MathDoubleMul(B=0.5, A=@Add.Result)
   RigVMFunction_AnimEvalRichCurve = &RigVMFunction_AnimEvalRichCurve(
    Curve="(EditorCurveData=(Keys=((InterpMode=RCIM_Cubic),(InterpMode=RCIM_Cubic,TangentMode=RCTM_User,Time=0.8427,Value=0.3264,ArriveTangent=1.7161,LeaveTangent=1.7161),(InterpMode=RCIM_Cubic,TangentMode=RCTM_User,Time=1,Value=1,ArriveTangent=3.1792,LeaveTangent=3. ...(truncated)"
    SourceMinimum=0
    TargetMinimum=0
    Value=@Multiply.Result
   )
   n5 = RigVMVariableNode() {variable=SPR Upperarm R Raised}
   n6 = RigVMVariableNode() {variable=SPR Clavicle R Raise Side}
   n7 = &RigVMFunction_AnimEvalRichCurve(
    Curve="(EditorCurveData=(Keys=((InterpMode=RCIM_Cubic),(InterpMode=RCIM_Cubic,TangentMode=RCTM_User,Time=0.8427,Value=0.3264,ArriveTangent=1.7161,LeaveTangent=1.7161),(InterpMode=RCIM_Cubic,TangentMode=RCTM_User,Time=1,Value=1,ArriveTangent=3.1792,LeaveTangent=3. ...(truncated)"
    SourceMinimum=0
    TargetMinimum=0
    Value=&RigVMFunction_MathDoubleMul(B=0.5, A=&RigVMFunction_MathDoubleAdd(B=@n5.Value, A=@n6.Value).Result).Result
   )
   > Modify Transforms with Mirror_6 = RigVMFunctionReferenceNode(
    Base Modify Data="((Item=(Type=Bone,Name=\"dyn_hood_rear_a_l_base\"),Transform=(Rotation=(X=-0.2569,Y=0.2722,Z=0.3349,W=0.8647),Translation=(X=-2,Y=2,Z=0),Scale3D=(X=1,Y=1,Z=1))),(Item=(Type=Bone,Name=\"dyn_hood_rear_b_l_base\"),Transform=(Rotation=(X=-0.1663,Y=0.1173,Z=-0.0347 ...(truncated)"
    Mirror Modify Data="((Item=(Type=Bone,Name=\"dyn_hood_rear_a_r_base\"),Transform=(Rotation=(X=0,Y=0,Z=0,W=1),Translation=(X=0,Y=0,Z=0),Scale3D=(X=1,Y=1,Z=1))),(Item=(Type=Bone,Name=\"dyn_hood_rear_b_r_base\"),Transform=(Rotation=(X=0,Y=0,Z=0,W=1),Translation=(X=0,Y=0,Z=0),Scale3D ...(truncated)"
    Weight=@RigVMFunction_AnimEvalRichCurve.Result
    "Mirror Weight "=@n7.Result
   )
   > Modify Transforms with Mirror = RigVMFunctionReferenceNode(
    Base Modify Data="((Item=(Type=Bone,Name=\"dyn_hood_front_a_l_base\"),Transform=(Rotation=(X=-0.2896,Y=0.2028,Z=0.228,W=0.9072),Translation=(X=-2,Y=-2,Z=-3),Scale3D=(X=1,Y=1,Z=1))),(Item=(Type=Bone,Name=\"dyn_hood_front_b_l_base\"),Transform=(Rotation=(X=-0.0076,Y=0.0868,Z=0.08 ...(truncated)"
    Mirror Modify Data="((Item=(Type=Bone,Name=\"dyn_hood_front_a_r_base\"),Transform=(Rotation=(X=0,Y=0,Z=0,W=1),Translation=(X=0,Y=0,Z=0),Scale3D=(X=1,Y=1,Z=1))),(Item=(Type=Bone,Name=\"dyn_hood_front_b_r_base\"),Transform=(Rotation=(X=0,Y=0,Z=0,W=1),Translation=(X=0,Y=0,Z=0),Scale ...(truncated)"
    Weight=RigVMVariableNode() {variable=SPR Clavice L Fwd Raise}.Value
    "Mirror Weight "=RigVMVariableNode() {variable=SPR Clavicle R Fwd Raise}.Value
   )
   > n8 = &RigUnit_ModifyTransforms(
    ItemToModify="((Item=(Type=Bone,Name=\"dyn_hood_front_a_l_base\"),Transform=(Rotation=(X=0,Y=0,Z=0.2164,W=0.9763),Translation=(X=-3,Y=3,Z=0),Scale3D=(X=1,Y=1,Z=1))),(Item=(Type=Bone,Name=\"dyn_hood_front_b_l_base\"),Transform=(Rotation=(X=0,Y=0,Z=0.2164,W=0.9763),Translatio ...(truncated)"
    WeightMinimum=0
    Weight=&RigVMFunction_KalmanFloat(BufferSize=4, Value=RigVMVariableNode() {variable=SPR Head Fwd Left Down}.Value) {template="KalmanFilter::Execute(in Value,in BufferSize,out Result)"}.Result
   )
   RigVMFunction_ControlFlowBranch = &RigVMFunction_ControlFlowBranch(ExecuteContext=@n8.ExecutePin, Condition=RigVMVariableNode() {variable=IsFrontendIdle}.Value)
   &RigUnit_ModifyTransforms(
    ItemToModify="((Item=(Type=Bone,Name=\"dyn_hood_front_a_l_base\"),Transform=(Rotation=(X=-0.2113,Y=0.366,Z=0.4532,W=0.7849),Translation=(X=0,Y=2.5,Z=0),Scale3D=(X=1,Y=1,Z=1))),(Item=(Type=Bone,Name=\"dyn_hood_base\"),Transform=(Rotation=(X=0,Y=0,Z=0,W=1),Translation=(X=0.7, ...(truncated)"
    WeightMinimum=0
    ExecutePin=@RigVMFunction_ControlFlowBranch.True
    Weight=RigVMVariableNode() {variable=SPR Head Side Clav Raise L}.Value
   )
   n9 = &RigVMFunction_AnimEvalRichCurve(
    Curve=(EditorCurveData=(Keys=((),(Time=0.6273,Value=0.7979),(Time=1,Value=1))))
    SourceMinimum=0
    TargetMinimum=0
    Value=RigVMVariableNode() {variable=SPR Upperarm R Frontend}.Value
   )
   > &RigUnit_ModifyTransforms(
    ItemToModify="((Item=(Type=Bone,Name=\"dyn_hood_front_a_r_base\"),Transform=(Rotation=(X=-0.3394,Y=-0.1767,Z=0.4266,W=0.8195),Translation=(X=3,Y=3,Z=0),Scale3D=(X=1,Y=1,Z=1))),(Item=(Type=Bone,Name=\"dyn_hood_front_b_r_base\"),Transform=(Rotation=(X=0,Y=0,Z=0.2588,W=0.9659) ...(truncated)"
    WeightMinimum=0
    Weight=@n9.Result
   )
   n10 = &RigUnit_SphericalPoseReader(
    ActiveRegionScaleFactors=(PositiveWidth=1,NegativeWidth=4,PositiveHeight=4,NegativeHeight=2)
    ActiveRegionSize=0.1
    Debug=(bDrawDebug=false,DebugScale=25,DebugSegments=20,DebugThickness=0.25)
    DriverAxis=(X=1,Y=0,Z=0)
    DriverItem=(Type=Bone,Name="upperarm_l")
    FalloffRegionScaleFactors=(PositiveWidth=1,NegativeWidth=1,PositiveHeight=1,NegativeHeight=1)
    FalloffSize=0.4
    OptionalParentItem=(Type=Bone,Name="spine_05")
    RotationOffset=(X=30,Y=-10,Z=70)
    ExecutePin=@RigVMFunction_Sequence.A
   )
   > RigVMVariableNode(Value=@n10.OutputParam) {variable=SPR Upperarm L Raised}
   > n11 = &RigUnit_SphericalPoseReader(
    ActiveRegionScaleFactors=(PositiveWidth=1,NegativeWidth=4,PositiveHeight=4,NegativeHeight=2)
    ActiveRegionSize=0.1
    Debug=(bDrawDebug=false,DebugScale=25,DebugSegments=20,DebugThickness=0.25)
    DriverAxis=(X=-1,Y=0,Z=0)
    DriverItem=(Type=Bone,Name="upperarm_r")
    FalloffRegionScaleFactors=(PositiveWidth=1,NegativeWidth=1,PositiveHeight=1,NegativeHeight=1)
    FalloffSize=0.4
    FlipHeightScaling=true
    OptionalParentItem=(Type=Bone,Name="spine_05")
    RotationOffset=(X=150,Y=10,Z=-110)
   )
   > RigVMVariableNode(Value=@n11.OutputParam) {variable=SPR Upperarm R Raised}
   > n12 = &RigUnit_SphericalPoseReader(
    ActiveRegionScaleFactors=(PositiveWidth=0.5,NegativeWidth=3,PositiveHeight=1.5,NegativeHeight=1.5)
    ActiveRegionSize=0.1
    Debug=(bDrawDebug=false,DebugScale=25,DebugSegments=20,DebugThickness=0.25)
    DriverAxis=(X=1,Y=0,Z=0)
    DriverItem=(Type=Bone,Name="upperarm_l")
    FalloffRegionScaleFactors=(PositiveWidth=1,NegativeWidth=1,PositiveHeight=1,NegativeHeight=1)
    FalloffSize=0.4
    OptionalParentItem=(Type=Bone,Name="spine_05")
    RotationOffset=(X=90,Y=0,Z=0)
   )
   > RigVMVariableNode(Value=@n12.OutputParam) {variable=SPR Upperarm L Fwd}
   > n13 = &RigUnit_SphericalPoseReader(
    ActiveRegionScaleFactors=(PositiveWidth=0.5,NegativeWidth=3,PositiveHeight=1.5,NegativeHeight=1.5)
    ActiveRegionSize=0.1
    Debug=(bDrawDebug=false,DebugScale=25,DebugSegments=20,DebugThickness=0.25)
    DriverAxis=(X=-1,Y=0,Z=0)
    DriverItem=(Type=Bone,Name="upperarm_r")
    FalloffRegionScaleFactors=(PositiveWidth=1,NegativeWidth=1,PositiveHeight=1,NegativeHeight=1)
    FalloffSize=0.4
    FlipWidthScaling=true
    OptionalParentItem=(Type=Bone,Name="spine_05")
    RotationOffset=(X=90,Y=180,Z=180)
   )
   > RigVMVariableNode(Value=@n13.OutputParam) {variable=SPR Upperarm R Fwd}
   > n14 = &RigUnit_SphericalPoseReader(
    ActiveRegionScaleFactors=(PositiveWidth=1,NegativeWidth=5,PositiveHeight=3,NegativeHeight=3)
    ActiveRegionSize=0.1
    Debug=(bDrawDebug=false,bDraw2D=False,bDrawLocalAxes=False,DebugScale=25,DebugSegments=20,DebugThickness=0.25)
    DriverAxis=(X=1,Y=0,Z=0)
    DriverItem=(Type=Bone,Name="head")
    FalloffRegionScaleFactors=(PositiveWidth=1,NegativeWidth=1,PositiveHeight=1,NegativeHeight=1)
    FalloffSize=0.5
    FlipHeightScaling=False
    FlipWidthScaling=False
    OptionalParentItem=(Type=Bone,Name="spine_05")
    RotationOffset=(X=-90,Y=0,Z=0)
   )
   > RigVMVariableNode(Value=@n14.OutputParam) {variable=SPR Head Back}
   > n15 = &RigUnit_SphericalPoseReader(
    ActiveRegionScaleFactors=(PositiveWidth=1,NegativeWidth=3,PositiveHeight=3,NegativeHeight=3)
    ActiveRegionSize=0.1
    Debug=(bDrawDebug=true,bDraw2D=False,bDrawLocalAxes=False,DebugScale=25,DebugSegments=20,DebugThickness=0.25)
    DriverAxis=(X=1,Y=0,Z=0)
    DriverItem=(Type=Bone,Name="clavicle_l")
    FalloffRegionScaleFactors=(PositiveWidth=1,NegativeWidth=1,PositiveHeight=1,NegativeHeight=1)
    FalloffSize=0.3
    FlipHeightScaling=False
    FlipWidthScaling=False
    OptionalParentItem=(Type=Bone,Name="spine_04")
    RotationOffset=(X=-60,Y=0,Z=45)
   )
   > RigVMVariableNode(Value=@n15.OutputParam) {variable=SPR Clavicle L Back}
   > n16 = &RigUnit_SphericalPoseReader(
    ActiveRegionScaleFactors=(PositiveWidth=1,NegativeWidth=3,PositiveHeight=3,NegativeHeight=3)
    ActiveRegionSize=0.1
    Debug=(bDrawDebug=true,bDraw2D=False,bDrawLocalAxes=False,DebugScale=25,DebugSegments=20,DebugThickness=0.25)
    DriverAxis=(X=-1,Y=0,Z=0)
    DriverItem=(Type=Bone,Name="clavicle_r")
    FalloffRegionScaleFactors=(PositiveWidth=1,NegativeWidth=1,PositiveHeight=1,NegativeHeight=1)
    FalloffSize=0.3
    FlipHeightScaling=False
    FlipWidthScaling=true
    OptionalParentItem=(Type=Bone,Name="spine_04")
    RotationOffset=(X=120,Y=0,Z=45)
   )
   > RigVMVariableNode(Value=@n16.OutputParam) {variable=SPR Clavicle R Back}
   > n17 = &RigUnit_SphericalPoseReader(
    ActiveRegionScaleFactors=(PositiveWidth=1,NegativeWidth=3,PositiveHeight=3,NegativeHeight=3)
    ActiveRegionSize=0.1
    Debug=(bDrawDebug=false,bDraw2D=False,bDrawLocalAxes=False,DebugScale=25,DebugSegments=20,DebugThickness=0.25)
    DriverAxis=(X=1,Y=0,Z=0)
    DriverItem=(Type=Bone,Name="clavicle_l")
    FalloffRegionScaleFactors=(PositiveWidth=1,NegativeWidth=1,PositiveHeight=1,NegativeHeight=1)
    FalloffSize=0.3
    FlipHeightScaling=False
    FlipWidthScaling=False
    OptionalParentItem=(Type=Bone,Name="None")
    RotationOffset=(X=0,Y=-35,Z=0)
   )
   > RigVMVariableNode(Value=@n17.OutputParam) {variable=SPR Clavicle L Raise Side}
   > n18 = &RigUnit_SphericalPoseReader(
    ActiveRegionScaleFactors=(PositiveWidth=1,NegativeWidth=3,PositiveHeight=3,NegativeHeight=3)
    ActiveRegionSize=0.1
    Debug=(bDrawDebug=false,bDraw2D=False,bDrawLocalAxes=False,DebugScale=25,DebugSegments=20,DebugThickness=0.25)
    DriverAxis=(X=-1,Y=0,Z=0)
    DriverItem=(Type=Bone,Name="clavicle_r")
    FalloffRegionScaleFactors=(PositiveWidth=1,NegativeWidth=1,PositiveHeight=1,NegativeHeight=1)
    FalloffSize=0.3
    FlipHeightScaling=False
    FlipWidthScaling=true
    OptionalParentItem=(Type=Bone,Name="None")
    RotationOffset=(X=180,Y=-35,Z=0)
   )
   > RigVMVariableNode(Value=@n18.OutputParam) {variable=SPR Clavicle R Raise Side}
   > n19 = &RigUnit_SphericalPoseReader(
    ActiveRegionScaleFactors=(PositiveWidth=1,NegativeWidth=3,PositiveHeight=3,NegativeHeight=1.5)
    ActiveRegionSize=0.1
    Debug=(bDrawDebug=false,bDraw2D=False,bDrawLocalAxes=False,DebugScale=25,DebugSegments=20,DebugThickness=0.25)
    DriverAxis=(X=1,Y=0,Z=0)
    DriverItem=(Type=Bone,Name="clavicle_l")
    FalloffRegionScaleFactors=(PositiveWidth=1,NegativeWidth=1,PositiveHeight=1,NegativeHeight=1)
    FalloffSize=0.3
    FlipHeightScaling=False
    FlipWidthScaling=False
    OptionalParentItem=(Type=Bone,Name="None")
    RotationOffset=(X=45,Y=-55,Z=0)
   )
   > RigVMVariableNode(Value=@n19.OutputParam) {variable=SPR Clavice L Fwd Raise}
   > n20 = &RigUnit_SphericalPoseReader(
    ActiveRegionScaleFactors=(PositiveWidth=1,NegativeWidth=3,PositiveHeight=3,NegativeHeight=1.5)
    ActiveRegionSize=0.1
    Debug=(bDrawDebug=false,bDraw2D=False,bDrawLocalAxes=False,DebugScale=25,DebugSegments=20,DebugThickness=0.25)
    DriverAxis=(X=-1,Y=0,Z=0)
    DriverItem=(Type=Bone,Name="clavicle_r")
    FalloffRegionScaleFactors=(PositiveWidth=1,NegativeWidth=1,PositiveHeight=1,NegativeHeight=1)
    FalloffSize=0.3
    FlipHeightScaling=true
    FlipWidthScaling=False
    OptionalParentItem=(Type=Bone,Name="None")
    RotationOffset=(X=-45,Y=135,Z=0)
   )
   > RigVMVariableNode(Value=@n20.OutputParam) {variable=SPR Clavicle R Fwd Raise}
   > n21 = &RigUnit_SphericalPoseReader(
    ActiveRegionScaleFactors=(PositiveWidth=1,NegativeWidth=1,PositiveHeight=1,NegativeHeight=1)
    ActiveRegionSize=0.08
    Debug=(bDrawDebug=false,bDraw2D=False,bDrawLocalAxes=False,DebugScale=25,DebugSegments=20,DebugThickness=0.25)
    DriverAxis=(X=0,Y=1,Z=0)
    DriverItem=(Type=Bone,Name="head")
    FalloffRegionScaleFactors=(PositiveWidth=1,NegativeWidth=1,PositiveHeight=1,NegativeHeight=1)
    FalloffSize=0.15
    FlipHeightScaling=False
    FlipWidthScaling=False
    OptionalParentItem=(Type=Bone,Name="spine_05")
    RotationOffset=(X=120,Y=-45,Z=0)
   )
   > RigVMVariableNode(Value=@n21.OutputParam) {variable=SPR Head Fwd Left Down}
   > n22 = &RigUnit_SphericalPoseReader(
    ActiveRegionScaleFactors=(PositiveWidth=1,NegativeWidth=1,PositiveHeight=1,NegativeHeight=1)
    ActiveRegionSize=0.1
    Debug=(bDrawDebug=true,bDraw2D=False,bDrawLocalAxes=False,DebugScale=25,DebugSegments=20,DebugThickness=0.25)
    DriverAxis=(X=0,Y=1,Z=0)
    DriverItem=(Type=Bone,Name="head")
    FalloffRegionScaleFactors=(PositiveWidth=1,NegativeWidth=1,PositiveHeight=1,NegativeHeight=1)
    FalloffSize=0.2
    FlipHeightScaling=False
    FlipWidthScaling=False
    OptionalParentItem=(Type=Bone,Name="clavicle_l")
    RotationOffset=(X=160,Y=0,Z=0)
   )
   > RigVMVariableNode(Value=@n22.OutputParam) {variable=SPR Head Side Clav Raise L}
   > n23 = &RigUnit_SphericalPoseReader(
    ActiveRegionScaleFactors=(PositiveWidth=1,NegativeWidth=1,PositiveHeight=1,NegativeHeight=1)
    ActiveRegionSize=0.1
    Debug=(bDrawDebug=false,bDraw2D=False,bDrawLocalAxes=False,DebugScale=25,DebugSegments=20,DebugThickness=0.25)
    DriverAxis=(X=-1,Y=0,Z=0)
    DriverItem=(Type=Bone,Name="upperarm_r")
    FalloffRegionScaleFactors=(PositiveWidth=1,NegativeWidth=1,PositiveHeight=1,NegativeHeight=1)
    FalloffSize=0.3
    FlipHeightScaling=False
    FlipWidthScaling=False
    OptionalParentItem=(Type=Bone,Name="spine_05")
    RotationOffset=(X=110,Y=0,Z=160)
   )
   > RigVMVariableNode(Value=@n23.OutputParam) {variable=SPR Upperarm R Frontend}
   &RigUnit_ParentConstraint(
    AdvancedSettings=(RotationOrderForFilter=XZY)
    Child=(Type=Bone,Name="dyn_hood_front_a_l_base")
    ExecutePin=()
    Parents="((Item=(Type=Bone,Name=\"deform_upperarm_up_l\"),Weight=0.0147),(Item=(Type=Bone,Name=\"upperarm_twist_01_l\"),Weight=0.0223),(Item=(Type=Bone,Name=\"deform_pec_l\"),Weight=0.0225),(Item=(Type=Bone,Name=\"dyn_collar_l\"),Weight=0.0311),(Item=(Type=Bone,Name=\"spine ...(truncated)"
    Weight=1
   )
   &RigUnit_ParentConstraint(
    AdvancedSettings=(RotationOrderForFilter=XZY)
    Child=(Type=Bone,Name="dyn_hood_front_a_r_base")
    ExecutePin=()
    Parents=((Item=(Type=Bone,Name="head"),Weight=1.0))
    Weight=1
   )
  ]}
]
hierarchy [
 root {loc=<0 0 0>, rot=<0 0 0 1>}
 pelvis {parent=root, loc=<0 -0.7071 93.1191>, rot=<0.0127 0.707 -0.0127 -0.707>}
 spine_01 {parent=pelvis, loc=<6.7494 -0.6388 0>, rot=<0 0 0.034 -0.9994>}
 spine_02 {parent=spine_01, loc=<7.2518 -0.035 0>, rot=<0 0 0.0238 -0.9997>}
 spine_03 {parent=spine_02, loc=<6.0104 0 0>, rot=<0 0 0.0149 -0.9999>}
 spine_04 {parent=spine_03, loc=<8.1024 0 0>, rot=<0 0 -0.0132 -0.9999>}
 spine_05 {parent=spine_04, loc=<10.0146 0 0>, rot=<0 0 -0.0015 -1>}
 neck_01 {parent=spine_05, loc=<11.3138 0.1529 0>, rot=<0 0 -0.2003 -0.9797>}
 neck_02 {parent=neck_01, loc=<4.5627 0 0>, rot=<0 0 0.0504 -0.9987>}
 head {parent=neck_02, loc=<5.3914 0 0>, rot=<0 0 0.115 -0.9934>}
 dyn_hood_base {parent=neck_02, loc=<5.3914 0 0>, rot=<0 0 0.115 -0.9934>}
 dyn_hood_front_a_l_base {parent=dyn_hood_base, loc=<-6.5244 3.6961 -11.9176>, rot=<-0.1077 -0.9347 0.318 -0.1166>}
 dyn_hood_front_a_01_l {parent=dyn_hood_front_a_l_base, loc=<0 0 0>, rot=<0 0 0 1>}
 dyn_hood_front_a_02_l {parent=dyn_hood_front_a_01_l, loc=<3.7749 0 0>, rot=<0 0 0 1>}
 dyn_hood_front_a_03_l {parent=dyn_hood_front_a_02_l, loc=<3.5 0 0>, rot=<0 0 0 1>}
 dyn_hood_front_b_l_base {parent=dyn_hood_base, loc=<-9.0366 8.5499 -8.0923>, rot=<-0.0631 -0.9679 0.224 -0.0949>}
 dyn_hood_front_b_01_l {parent=dyn_hood_front_b_l_base, loc=<0 0 0>, rot=<0 0 0 1>}
 dyn_hood_front_b_02_l {parent=dyn_hood_front_b_01_l, loc=<4.4925 0 0>, rot=<0 0 0 1>}
 dyn_hood_front_b_03_l {parent=dyn_hood_front_b_02_l, loc=<3.1411 0 0>, rot=<0 0 0 1>}
 dyn_hood_rear_a_l_base {parent=dyn_hood_base, loc=<-5.6701 -6.2092 -12.3401>, rot=<-0.1427 -0.5021 0.8126 -0.2593>}
 dyn_hood_rear_a_01_l {parent=dyn_hood_rear_a_l_base, loc=<0 0 0>, rot=<0 0 0.0634 -0.998>}
 dyn_hood_rear_a_02_l {parent=dyn_hood_rear_a_01_l, loc=<3.0998 0 0>, rot=<0 0 0 1>}
 dyn_hood_rear_b_l_base {parent=dyn_hood_base, loc=<-7.6129 -11.1257 -7.705>, rot=<-0.0672 -0.1831 0.9676 -0.1604>}
 dyn_hood_rear_b_01_l {parent=dyn_hood_rear_b_l_base, loc=<0 0 0>, rot=<0 0 0 1>}
 dyn_hood_rear_b_02_l {parent=dyn_hood_rear_b_01_l, loc=<4.8289 0 0>, rot=<0 0 0 1>}
 dyn_hood_rear_a_r_base {parent=dyn_hood_base, loc=<-5.6701 -6.2092 12.3401>, rot=<0.5021 -0.1427 -0.2593 -0.8126>}
 dyn_hood_rear_a_01_r {parent=dyn_hood_rear_a_r_base, loc=<0 0 0>, rot=<0 0 0 -1>}
 dyn_hood_rear_a_02_r {parent=dyn_hood_rear_a_01_r, loc=<-3.1 0.0002 0>, rot=<0 0 0 -1>}
 dyn_hood_rear_b_r_base {parent=dyn_hood_base, loc=<-7.6125 -11.1257 7.705>, rot=<0.1831 -0.0672 -0.1604 -0.9676>}
 dyn_hood_rear_b_01_r {parent=dyn_hood_rear_b_r_base, loc=<0 0 0>, rot=<0 0 0 -1>}
 dyn_hood_rear_b_02_r {parent=dyn_hood_rear_b_01_r, loc=<-4.8288 -0.0001 0>, rot=<0 0 0 1>}
 dyn_hood_front_a_r_base {parent=dyn_hood_base, loc=<-6.524 3.6961 11.9176>, rot=<0.9347 -0.1077 -0.1166 -0.318>}
 dyn_hood_front_a_01_r {parent=dyn_hood_front_a_r_base, loc=<0 0 0>, rot=<0 0 0 1>}
 dyn_hood_front_a_02_r {parent=dyn_hood_front_a_01_r, loc=<-3.7752 0.0001 0.0001>, rot=<0 0 0 1>}
 dyn_hood_front_a_03_r {parent=dyn_hood_front_a_02_r, loc=<-3.5 0 0>, rot=<0 0 0 1>}
 dyn_hood_front_b_r_base {parent=dyn_hood_base, loc=<-9.0362 8.5499 8.0923>, rot=<0.9679 -0.0631 -0.0949 -0.224>}
 dyn_hood_front_b_01_r {parent=dyn_hood_front_b_r_base, loc=<0 0 0>, rot=<0 0 0 -1>}
 dyn_hood_front_b_02_r {parent=dyn_hood_front_b_01_r, loc=<-4.4928 0.0001 0>, rot=<0 0 0 1>}
 dyn_hood_front_b_03_r {parent=dyn_hood_front_b_02_r, loc=<-3.1408 -0.0001 0>, rot=<0 0 0 1>}
 dyn_hood_front_c_l_base {parent=dyn_hood_base, loc=<-10.1072 10.8034 -3.0514>, rot=<-0.0582 -0.9899 0.1249 -0.0341>}
 dyn_hood_front_c_01_l {parent=dyn_hood_front_c_l_base, loc=<0 0 0>, rot=<0 0 0 1>}
 dyn_hood_front_c_02_l {parent=dyn_hood_front_c_01_l, loc=<4.4925 0 0>, rot=<0 0 0 1>}
 dyn_hood_front_c_03_l {parent=dyn_hood_front_c_02_l, loc=<3.1411 0 0>, rot=<0 0 0 1>}
 dyn_hood_front_c_r_base {parent=dyn_hood_base, loc=<-10.1069 10.8034 3.0514>, rot=<0.9899 -0.0582 -0.0341 -0.1249>}
 dyn_hood_front_c_01_r {parent=dyn_hood_front_c_r_base, loc=<0 0 0>, rot=<0 0 0 1>}
 dyn_hood_front_c_02_r {parent=dyn_hood_front_c_01_r, loc=<-4.493 0.0001 0>, rot=<0 0 0 1>}
 dyn_hood_front_c_03_r {parent=dyn_hood_front_c_02_r, loc=<-3.1405 0 0>, rot=<0 0 0 1>}
 dyn_hood_rear_mid_base {parent=dyn_hood_base, loc=<-8.6064 -12.9167 0>, rot=<0 0 0.99 -0.1412>}
 dyn_hood_rear_mid_01 {parent=dyn_hood_rear_mid_base, loc=<0 0 0>, rot=<0 0 0 1>}
 dyn_hood_rear_mid_02 {parent=dyn_hood_rear_mid_01, loc=<4.8289 0 0>, rot=<0 0 0 1>}
 clavicle_l {parent=spine_05, loc=<4.3575 -0.1466 -2.3223>, rot=<0.0465 -0.7289 -0.009 -0.6829>}
 upperarm_l {parent=clavicle_l, loc=<15.5111 0 0>, rot=<0.0226 -0.3234 -0.0088 -0.9459>}
 upperarm_twist_01_l {parent=upperarm_l, loc=<0.0041 0 0>, rot=<0 0 0 -1>}
 deform_upperarm_fr_base_l {parent=upperarm_twist_01_l, loc=<-0.3691 7.0506 -2.765>, rot=<0 0 0 -1>}
 deform_upperarm_fr_l {parent=deform_upperarm_fr_base_l, loc=<0 0 0>, rot=<0 0 0 1>}
 deform_upperarm_up_base_l {parent=upperarm_twist_01_l, loc=<-3.2216 -1.1391 3.3311>, rot=<0 0 0 1>}
 deform_upperarm_up_l {parent=deform_upperarm_up_base_l, loc=<0 0 0>, rot=<0 0 0 1>}
 clavicle_r {parent=spine_05, loc=<4.3575 -0.1466 2.3223>, rot=<0.7289 0.0465 -0.6829 0.009>}
 upperarm_r {parent=clavicle_r, loc=<-15.5111 0 0>, rot=<0.0226 -0.3234 -0.0088 -0.9459>}
 upperarm_twist_01_r {parent=upperarm_r, loc=<-0.0041 0 0>, rot=<0 0 0 -1>}
 deform_upperarm_fr_base_r {parent=upperarm_twist_01_r, loc=<0.3689 -7.0506 2.7652>, rot=<0 0 0 -1>}
 deform_upperarm_fr_r {parent=deform_upperarm_fr_base_r, loc=<0 0 0>, rot=<0 0 0 1>}
 deform_upperarm_up_base_r {parent=upperarm_twist_01_r, loc=<3.2214 1.1391 -3.3308>, rot=<0 0 0 1>}
 deform_upperarm_up_r {parent=deform_upperarm_up_base_r, loc=<0 0 0>, rot=<0 0 0 -1>}
 upperarm_twist_02_r {parent=upperarm_r, loc=<-13.7941 0 0>, rot=<0 0 0 -1>}
 dyn_collar_l {parent=spine_05, loc=<6.9728 7.8136 -5.6404>, rot=<-0.4595 -0.6117 0.0281 -0.6433>}
 dyn_collar_r {parent=spine_05, loc=<6.9733 7.8136 5.6404>, rot=<-0.6117 0.4595 0.6433 0.0281>}
 deform_pec_aim_base_l {parent=spine_04, loc=<7.8732 12.6388 -2.4123>, rot=<0.0102 -0.4602 0.3979 -0.7936>}
 deform_pec_aim_l {parent=deform_pec_aim_base_l, loc=<0 0 0>, rot=<0 0 0 1>}
 deform_pec_base_l {parent=deform_pec_aim_l, loc=<10.766 0 0>, rot=<0 0 0 1>}
 deform_pec_l {parent=deform_pec_base_l, loc=<0.055 0 0>, rot=<0 0 0 1>}
 deform_pec_aim_base_r {parent=spine_04, loc=<7.8727 12.6388 2.4123>, rot=<-0.4602 -0.0102 0.7936 0.3979>}
 deform_pec_aim_r {parent=deform_pec_aim_base_r, loc=<0 0 0>, rot=<0 0 0 1>}
 deform_pec_base_r {parent=deform_pec_aim_r, loc=<-10.7662 -0.0006 -0.0007>, rot=<0 0 0 1>}
 deform_pec_r {parent=deform_pec_base_r, loc=<0 0 0>, rot=<0 0 0 1>}
]
variables [
 IsFrontendIdle {type=bool, default=False}
 SPR Clavice L Fwd Raise {type=float, default=0}
 SPR Clavicle L Back {type=float, default=0}
 SPR Clavicle L Raise Side {type=float, default=0}
 SPR Clavicle R Back {type=float, default=0}
 SPR Clavicle R Fwd Raise {type=float, default=0}
 SPR Clavicle R Raise Side {type=float, default=0}
 SPR Head Back {type=float, default=0}
 SPR Head Fwd Left Down {type=float, default=0}
 SPR Head Side Clav Raise L {type=float, default=0}
 SPR Upperarm L Fwd {type=float, default=0}
 SPR Upperarm L Raised {type=float, default=0}
 SPR Upperarm R Frontend {type=float, default=0}
 SPR Upperarm R Fwd {type=float, default=0}
 SPR Upperarm R Raised {type=float, default=0}
]
